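# Sound cue for the SunRose wings jump foley in flying gameplay, playing one of two jump wave variations at random with equal weight.
# A branch on the "islocal" parameter feeds the same random node on all three outputs, so local and non-local playback currently sound identical.
# It uses the shared player foley sound class and basic foley attenuation, with a concurrency override allowing a single instance.
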SoundCue /SunRoseFlyingGameplay/Sounds/Foley_SunRose_Wings_Jump_Cue
root=Branch_0
nodes [
 {id=Branch_0, children [Random_1,Random_1,Random_1], properties {boolParameterName=islocal}}
 {id=Random_1, children [WavePlayer_6,WavePlayer_7], properties {weights [1,1]}}
 {id=WavePlayer_6, properties {soundWaveAssetPtr=/SunRoseFlyingGameplay/Sounds/wav/FN_S29_OW_Jump_01}}
 {id=WavePlayer_7, properties {soundWaveAssetPtr=/SunRoseFlyingGameplay/Sounds/wav/FN_S29_OW_Jump_02}}
]
properties {
 attenuationSettings=&Foley_Attenuation_Basic_Att
 bOverrideConcurrency=true
 maxDistance=1536
 soundClassObject=&Fort_Player_Foley_Athena
 volumeMultiplier=0.8
 concurrencyOverrides {maxCount=1}
}
