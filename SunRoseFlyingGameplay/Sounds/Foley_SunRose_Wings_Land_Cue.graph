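# Sound cue for the SunRose wings landing foley, classed as player foley.
# Both playback paths share a random pick between two equally weighted landing waves.
# An islocal branch sends the local case through a SunRose Wings 2D attenuation.
# The non-local and unset cases use basic foley attenuation, then a modulator that adds a fixed volume boost.
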SoundCue /SunRoseFlyingGameplay/Sounds/Foley_SunRose_Wings_Land_Cue
root=Branch_0
nodes [
 {id=Attenuation_0, children [Random_0], properties {attenuationSettings=&/SunRose_Wings_2D_Att}}
 {
  id=Attenuation_1
  children [Random_0]
  properties {attenuationSettings=&Foley_Attenuation_Basic_Att}
 }
 {id=Branch_0, children [Attenuation_0, Modulator_0, Modulator_0], properties {boolParameterName=islocal}}
 {id=Modulator_0, children [Attenuation_1], properties {pitchMax=1, volumeMax=1.2, volumeMin=1.2}}
 {id=Random_0, children [WavePlayer_0,WavePlayer_5], properties {weights [1,1]}}
 {id=WavePlayer_0, properties {soundWaveAssetPtr=/SunRoseFlyingGameplay/Sounds/wav/FN_S29_OW_Landed_02}}
 {id=WavePlayer_5, properties {soundWaveAssetPtr=/SunRoseFlyingGameplay/Sounds/wav/FN_S29_OW_Landed_01}}
]
properties {
 maxDistance=3400
 soundClassObject=&Fort_Player_Foley_Athena
 volumeMultiplier=1
 concurrencyOverrides {maxCount=1}
}
