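# Stereo one-shot MetaSound Source for a camera change sound in SunRose flying gameplay, judging by its name.
# OnPlay triggers a Deluxe Stereo wave player patch, fed by the OneShot wave array (empty by default), whose left and right outputs drive the source.
# OnFinished is wired to the player's On Play output, so it signals at playback start rather than end.
# It uses SunRose wings attenuation and a melee weapon sound class.
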MetaSoundSource /SunRoseFlyingGameplay/Sounds/MSS_SunRose_CameraChange
inputs [
 OneShot {type=WaveAsset:Array}
 UE.Source.OnPlay {type=Trigger}
]
interfaces [UE.OutputFormat.Stereo:1.0, UE.Source.OneShot:1.0, UE.Source:1.1]
outputs [
 "UE.OutputFormat.Stereo.Audio:0" {type=Audio}
 "UE.OutputFormat.Stereo.Audio:1" {type=Audio}
 UE.Source.OneShot.OnFinished {type=Trigger}
]
pages [Default {
  input_defaults {UE.Source.OnPlay=null, OneShot []}
  nodes [
   InitVariable.Float()
   InitVariable.Float()
   InitVariable.Float()
   InitVariable.Float()
   MSP_Audio_WavePlayer_Deluxe_Stereo_0 = &MSP_Audio_WavePlayer_Deluxe_Stereo(Wave Asset=^OneShot, Play Trigger=^UE.Source.OnPlay)
  ]
  links [
   {dst="@UE.Source.OneShot.OnFinished", src=MSP_Audio_WavePlayer_Deluxe_Stereo_0.On Play}
   {dst="@UE.OutputFormat.Stereo.Audio:0", src=MSP_Audio_WavePlayer_Deluxe_Stereo_0.Out L}
   {dst="@UE.OutputFormat.Stereo.Audio:1", src=MSP_Audio_WavePlayer_Deluxe_Stereo_0.Out R}
  ]
 }]
properties {
 assetImportData=&self:AssetImportData
 attenuationSettings=&/SunRose_Wings_Att
 maxDistance=24300
 numChannels=2
 outputFormat=Stereo
 sampleRate=48000
 soundClassObject=&Fort_WeaponMelee_PSM
}
