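# Mono MetaSound source for third-person SunRose wing near-miss whooshes, playing a random fast-by sound as the emitter passes close to the listener.
# A threshold trigger on attenuation distance feeds a branch gated by a Speed comparison, which fires the random pick into a wave player.
# The air-loop array and distance MapRange do not reach the output.
# Concurrency allows one instance and prevents new ones.
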MetaSoundSource /SunRoseFlyingGameplay/Sounds/MSS_SunRose_Wings_3P_NearMiss
inputs [
 3P Bys {type=WaveAsset:Array}
 3PAirLoop {type=WaveAsset:Array}
 Speed {type=Float}
 UE.Attenuation.Distance {type=Float}
 UE.Source.OnPlay {type=Trigger}
 UE.Source.Orientation.Azimuth {type=Float}
 UE.Source.Orientation.Elevation {type=Float}
]
interfaces [UE.Attenuation:1.0, UE.OutputFormat.Mono:1.0, UE.Source.Orientation:1.0, UE.Source:1.1]
outputs ["UE.OutputFormat.Mono.Audio:0" {type=Audio}]
pages [Default {
  input_defaults {
   Speed=2000
   UE.Attenuation.Distance=0
   UE.Source.OnPlay=null
   UE.Source.Orientation.Azimuth=0
   UE.Source.Orientation.Elevation=0
   3P Bys [
    /SunRoseFlyingGameplay/Sounds/wav/FN_S29_OW_3P_Wing_Fast_By_01
    /SunRoseFlyingGameplay/Sounds/wav/FN_S29_OW_3P_Wing_Fast_By_02
    /SunRoseFlyingGameplay/Sounds/wav/FN_S29_OW_3P_Wing_Fast_By_03
   ]
   3PAirLoop [/SunRoseFlyingGameplay/Sounds/wav/FN_S29_OW_3P_lp_01, /SunRoseFlyingGameplay/Sounds/wav/FN_S29_OW_3P_lp_02]
  }
  nodes [
   InitVariable.Float()
   MapRange.MapRange:Float(In=UE.Reroute(Value=^UE.Attenuation.Distance).Value) {also {defaults {In Range B=3000, Out Range A=1, Out Range B=0}}}
   ArrayRandomGet_0 = Array.Random Get:WaveAsset:Array(
    In Array=^3P Bys
    Next=&MSP_Branch(
     Value=/FortMetasound/Patches/CompareFloat(A=^Speed) {also {defaults {B=1000}}}.Result
     In=UE.TriggerOnThreshold:Float(In=^UE.Attenuation.Distance) {also {defaults {Threshold=650, Type=1}}}.Out
    ).True
   )
   WavePlayer_0 = UE.Wave Player:Mono(Play=@ArrayRandomGet_0.On Next, Wave Asset=@ArrayRandomGet_0.Value)
  ]
  links [{dst="@UE.OutputFormat.Mono.Audio:0", src=WavePlayer_0.Out Mono}]
 }]
properties {
 assetImportData=&self:AssetImportData
 attenuationSettings=&/SunRose_Wings_NearMiss_Att
 bLooping=true
 bOverrideConcurrency=true
 duration=10000
 maxDistance=2200
 sampleRate=48000
 soundClassObject=&Fort_Gliders_InAir
 concurrencyOverrides {maxCount=1, resolutionRule=PreventNew, retriggerTime=5}
}
