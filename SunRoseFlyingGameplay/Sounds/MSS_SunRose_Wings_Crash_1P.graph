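# MetaSound source for the first-person SunRose wings crash sound, output in 5.1 with only front and rear pairs fed.
# On play, delayed attack envelopes start and fade in front and rear loops, with volume and EQ crossfade driven by Speed.
# A one-shot plays through a branch when isForcedCrash is false and mixes with the front loop, both EQ-crossfaded by EnergyRemaining.
# The rear layer skips the energy stage and outputs at half gain.
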MetaSoundSource /SunRoseFlyingGameplay/Sounds/MSS_SunRose_Wings_Crash_1P
inputs [
 EnergyRemaining {type=Float}
 isForcedCrash {type=Bool}
 Loop {type=WaveAsset:Array}
 OneShot {type=WaveAsset:Array}
 Rears {type=WaveAsset:Array}
 Speed {type=Float}
 UE.Source.OnPlay {type=Trigger}
]
interfaces [UE.OutputFormat.5dot1:1.0, UE.Source:1.1]
outputs [
 "UE.OutputFormat.5dot1.Audio:0" {type=Audio}
 "UE.OutputFormat.5dot1.Audio:1" {type=Audio}
 "UE.OutputFormat.5dot1.Audio:2" {type=Audio}
 "UE.OutputFormat.5dot1.Audio:3" {type=Audio}
 "UE.OutputFormat.5dot1.Audio:4" {type=Audio}
 "UE.OutputFormat.5dot1.Audio:5" {type=Audio}
]
pages [Default {
  input_defaults {
   EnergyRemaining=1
   Speed=3500
   UE.Source.OnPlay=null
   isForcedCrash=null
   Loop [/SunRoseFlyingGameplay/Sounds/wav/FN_S29_OW_ChargeBlast_lp_1P_01, /SunRoseFlyingGameplay/Sounds/wav/FN_S29_OW_ChargeBlast_lp_1P_02]
   OneShot [
    /SunRoseFlyingGameplay/Sounds/wav/FN_S29_OW_ChargeBlast_1P_LP_Start_01
    /SunRoseFlyingGameplay/Sounds/wav/FN_S29_OW_ChargeBlast_1P_LP_Start_02
    /SunRoseFlyingGameplay/Sounds/wav/FN_S29_OW_ChargeBlast_1P_LP_Start_03
   ]
   Rears [/SunRoseFlyingGameplay/Sounds/wav/FN_S29_OW_ChargeBlast_lp_Rears_02, /SunRoseFlyingGameplay/Sounds/wav/FN_S29_OW_ChargeBlast_lp_Rears_01]
  }
  nodes [
   TriggerDelay_0 = UE.Trigger Delay(In=^UE.Source.OnPlay)
   ADSREnvelopeADSREnvelope_1 = ADSR Envelope.ADSR Envelope:Audio(Trigger Attack=@TriggerDelay_0.out) {also {defaults {Attack Curve=1.2, Attack Time=1, Decay Time=0, Release Time=0, Sustain Level=1}}}
   n1 = &MSP_Audio_WavePlayer_Deluxe_Stereo(Wave Asset=^Loop, Play Trigger=@ADSREnvelopeADSREnvelope_1.On Attack Triggered) {also {defaults {
      Is Looping?="<MetaSoundBuilderResult.SUCCEEDED: 0>"
      "Random Pitch Max (Semitones)"=0
      "Random Pitch Min (Semitones)"=0
      WavePlayer Gain dB=-3
     }}}
   n2 = UE.Reroute(Value=UE.Multiply:Audio by Float(
     AdditionalOperands=/FortMetasound/SoundCue/MSP_ContinuousModulator(Value In=^Speed) {also {defaults {Input.Max=2000, Input.Min=500, Output.Max=1.2, Output.Min=0.8}}}.Value Out
     PrimaryOperand=UE.Reroute(Value=@ADSREnvelopeADSREnvelope_1.Out Envelope).Value
    ).out)
   n3 = /FortMetasound/Patches/MSP_Audio_EQCrossFade(
    Input=^Speed
    In L=UE.Multiply:Audio(AdditionalOperands=@n1.Out L, PrimaryOperand=@n2.Value).out
    In R=UE.Multiply:Audio(AdditionalOperands=@n1.Out R, PrimaryOperand=@n2.Value).out
   ) {also {defaults {Max.DbAdjustment=2, Max.Value=2000, Min.DbAdjustment=-2, Min.Value=1000}}}
   n4 = &MSP_Audio_EQCrossFade(Input=^EnergyRemaining, In L=@n3.Out L, In R=@n3.Out R)
   n5 = &MSP_Audio_WavePlayer_Deluxe_Stereo(Wave Asset=^OneShot, Play Trigger=&MSP_Branch(In=^UE.Source.OnPlay, Value=^isForcedCrash).False) {also {defaults {WavePlayer Gain dB=-3}}}
   n6 = &MSP_Audio_EQCrossFade(Input=^EnergyRemaining, In L=@n5.Out L, In R=@n5.Out R)
   AudioMixerAudioMixerStereo2_0 = "AudioMixer.Audio Mixer (Stereo, 2)"(In 0 L=@n4.Out L, In 0 R=@n4.Out R, In 1 L=@n6.Out L, In 1 R=@n6.Out R) {also {defaults {Gain 1=1.2}}}
   InitVariable.Float()
   InitVariable.Float()
   InitVariable.Float()
   InitVariable.Float()
   TriggerDelay_1 = UE.Trigger Delay(In=^UE.Source.OnPlay)
   ADSREnvelopeADSREnvelope_0 = ADSR Envelope.ADSR Envelope:Audio(Trigger Attack=@TriggerDelay_1.out) {also {defaults {Attack Curve=1.2, Attack Time=1, Decay Time=0, Release Time=0, Sustain Level=1}}}
   n7 = &MSP_Audio_WavePlayer_Deluxe_Stereo(Wave Asset=^Rears, Play Trigger=@ADSREnvelopeADSREnvelope_0.On Attack Triggered) {also {defaults {Is Looping?="<MetaSoundBuilderResult.SUCCEEDED: 0>", "Random Pitch Max (Semitones)"=0, "Random Pitch Min (Semitones)"=0}}}
   n8 = UE.Multiply:Audio by Float(
    AdditionalOperands=&MSP_ContinuousModulator(Value In=^Speed) {also {defaults {Input.Max=2000, Input.Min=500, Output.Max=1.2, Output.Min=0.8}}}.Value Out
    PrimaryOperand=UE.Reroute(Value=@ADSREnvelopeADSREnvelope_0.Out Envelope).Value
   )
   n9 = &MSP_Audio_EQCrossFade(
    Input=^Speed
    In L=UE.Multiply:Audio(AdditionalOperands=@n7.Out L, PrimaryOperand=@n8.out).out
    In R=UE.Multiply:Audio(AdditionalOperands=@n7.Out R, PrimaryOperand=@n8.out).out
   ) {also {defaults {Max.Value=2000, Min.DbAdjustment=-2, Min.Value=1000}}}
   Multiply_4 = UE.Multiply:Audio by Float(PrimaryOperand=@n9.Out L) {also {defaults {AdditionalOperands=0.5}}}
   Multiply_6 = UE.Multiply:Audio by Float(PrimaryOperand=@n9.Out R) {also {defaults {AdditionalOperands=0.5}}}
  ]
  links [
   {dst="@UE.OutputFormat.5dot1.Audio:0", src=AudioMixerAudioMixerStereo2_0.Out L}
   {dst="@UE.OutputFormat.5dot1.Audio:1", src=AudioMixerAudioMixerStereo2_0.Out R}
   {dst="@UE.OutputFormat.5dot1.Audio:4", src=Multiply_4.out}
   {dst="@UE.OutputFormat.5dot1.Audio:5", src=Multiply_6.out}
  ]
 }]
properties {
 assetImportData=&self:AssetImportData
 attenuationSettings=&/SunRose_ChargeBlast_Impact_Close_Att
 bLooping=true
 duration=10000
 maxDistance=12000
 numChannels=6
 outputFormat=FiveDotOne
 sampleRate=48000
 soundClassObject=&Fort_WeaponRanged
}
