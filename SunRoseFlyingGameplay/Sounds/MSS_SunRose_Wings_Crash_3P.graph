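# Looping stereo MetaSound for the third-person SunRose wings crash, built from ChargeBlast start one-shots and loops.
# Close and distant 3P layers crossfade by attenuation distance, beneath a stereo 1P layer that fades out with distance at close range.
# Delayed ADSR envelopes swell the loops in behind the starts.
# When isForcedCrash is true, close and 1P starts are skipped and the 3P loop envelope never triggers, leaving the distant start and 1P loop.
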MetaSoundSource /SunRoseFlyingGameplay/Sounds/MSS_SunRose_Wings_Crash_3P
inputs [
 1P_For_CloseRange_LP {type=WaveAsset:Array}
 1P_For_CloseRange_OS {type=WaveAsset:Array}
 3P_Close_Loop {type=WaveAsset:Array}
 3P_Close_OS {type=WaveAsset:Array}
 3P_Dist_OS {type=WaveAsset:Array}
 3P_Distant_Loop {type=WaveAsset:Array}
 isForcedCrash {type=Bool}
 UE.Attenuation.Distance {type=Float}
 UE.Source.OnPlay {type=Trigger}
]
interfaces [UE.Attenuation:1.0, UE.OutputFormat.Stereo:1.0, UE.Source:1.1]
outputs [
 "UE.OutputFormat.Stereo.Audio:0" {type=Audio}
 "UE.OutputFormat.Stereo.Audio:1" {type=Audio}
]
pages [Default {
  input_defaults {
   UE.Attenuation.Distance=240
   UE.Source.OnPlay=null
   isForcedCrash=null
   1P_For_CloseRange_LP [/SunRoseFlyingGameplay/Sounds/wav/FN_S29_OW_ChargeBlast_lp_1P_01, /SunRoseFlyingGameplay/Sounds/wav/FN_S29_OW_ChargeBlast_lp_1P_02]
   1P_For_CloseRange_OS [
    /SunRoseFlyingGameplay/Sounds/wav/FN_S29_OW_ChargeBlast_1P_LP_Start_01
    /SunRoseFlyingGameplay/Sounds/wav/FN_S29_OW_ChargeBlast_1P_LP_Start_02
    /SunRoseFlyingGameplay/Sounds/wav/FN_S29_OW_ChargeBlast_1P_LP_Start_03
    /SunRoseFlyingGameplay/Sounds/wav/FN_S29_OW_ChargeBlast_1P_LP_Start_04
   ]
   3P_Close_Loop [
    /SunRoseFlyingGameplay/Sounds/wav/FN_S29_OW_ChargeBlast_3P_LP_03
    /SunRoseFlyingGameplay/Sounds/wav/FN_S29_OW_ChargeBlast_3P_LP_02
    /SunRoseFlyingGameplay/Sounds/wav/FN_S29_OW_ChargeBlast_3P_LP_01
   ]
   3P_Close_OS [
    /SunRoseFlyingGameplay/Sounds/wav/FN_S29_OW_ChargeBlast_3P_LP_Start_01
    /SunRoseFlyingGameplay/Sounds/wav/FN_S29_OW_ChargeBlast_3P_LP_Start_02
    /SunRoseFlyingGameplay/Sounds/wav/FN_S29_OW_ChargeBlast_3P_LP_Start_03
    /SunRoseFlyingGameplay/Sounds/wav/FN_S29_OW_ChargeBlast_3P_LP_Start_04
   ]
   3P_Dist_OS [
    /SunRoseFlyingGameplay/Sounds/wav/FN_S29_OW_ChargeBlast_3P_LP_Dist_Start_03
    /SunRoseFlyingGameplay/Sounds/wav/FN_S29_OW_ChargeBlast_3P_LP_Dist_Start_02
    /SunRoseFlyingGameplay/Sounds/wav/FN_S29_OW_ChargeBlast_3P_LP_Dist_Start_01
   ]
   3P_Distant_Loop [/SunRoseFlyingGameplay/Sounds/wav/FN_S29_OW_ChargeBlast_3P_LP_Dist_01]
  }
  nodes [
   TriggerDelay_0 = UE.Trigger Delay(In=&MSP_Branch(In=^UE.Source.OnPlay, Value=^isForcedCrash).False)
   ADSREnvelopeADSREnvelope_0 = ADSR Envelope.ADSR Envelope:Audio(Trigger Attack=@TriggerDelay_0.out) {also {defaults {Attack Curve=1.2, Attack Time=2, Decay Time=0, Release Time=0, Sustain Level=1}}}
   ArrayRandomGet_2 = Array.Random Get:WaveAsset:Array(In Array=^3P_Close_Loop, Next=^UE.Source.OnPlay)
   WavePlayer_2 = UE.Wave Player:Mono(Play=@ArrayRandomGet_2.On Next, Wave Asset=@ArrayRandomGet_2.Value) {also {defaults {Loop="<MetaSoundBuilderResult.SUCCEEDED: 0>"}}}
   ArrayRandomGet_1 = Array.Random Get:WaveAsset:Array(In Array=^3P_Close_OS, Next=&MSP_Branch(In=^UE.Source.OnPlay, Value=^isForcedCrash).False) {also {defaults {Weights []}}}
   WavePlayer_0 = UE.Wave Player:Mono(Play=@ArrayRandomGet_1.On Next, Wave Asset=@ArrayRandomGet_1.Value)
   AudioMixerAudioMixerMono2_2 = "AudioMixer.Audio Mixer (Mono, 2)"(
    In 1=UE.Multiply:Audio(AdditionalOperands=@ADSREnvelopeADSREnvelope_0.Out Envelope, PrimaryOperand=@WavePlayer_2.Out Mono).out
    In 0=@WavePlayer_0.Out Mono
   ) {also {defaults {Gain 0=1.4, Gain 1=1.2}}}
   ArrayRandomGet_0 = Array.Random Get:WaveAsset:Array(In Array=^3P_Distant_Loop, Next=^UE.Source.OnPlay) {also {defaults {Weights []}}}
   WavePlayer_1 = UE.Wave Player:Mono(Play=@ArrayRandomGet_0.On Next, Wave Asset=@ArrayRandomGet_0.Value) {also {defaults {Loop="<MetaSoundBuilderResult.SUCCEEDED: 0>"}}}
   ArrayRandomGet_3 = Array.Random Get:WaveAsset:Array(In Array=^3P_Dist_OS, Next=^UE.Source.OnPlay) {also {defaults {Weights []}}}
   WavePlayer_3 = UE.Wave Player:Mono(Play=@ArrayRandomGet_3.On Next, Wave Asset=@ArrayRandomGet_3.Value)
   AudioMixerAudioMixerMono2_1 = "AudioMixer.Audio Mixer (Mono, 2)"(
    In 1=UE.Multiply:Audio(PrimaryOperand=@ADSREnvelopeADSREnvelope_0.Out Envelope, AdditionalOperands=@WavePlayer_1.Out Mono).out
    In 0=@WavePlayer_3.Out Mono
   ) {also {defaults {Gain 1=1.2}}}
   n1 = MapRange.MapRange:Float(In=^UE.Attenuation.Distance) {also {defaults {In Range A=2501, In Range B=13000}}}
   AudioMixerAudioMixerMono2_0 = "AudioMixer.Audio Mixer (Mono, 2)"(
    In 0=UE.Reroute(Value=@AudioMixerAudioMixerMono2_2.out).Value
    In 1=UE.Reroute(Value=@AudioMixerAudioMixerMono2_1.out).Value
    Gain 1=UE.Reroute(Value=@n1.Out Value).Value
    Gain 0=UE.Subtract:Float(AdditionalOperands=@n1.Out Value) {also {defaults {PrimaryOperand=1}}}.out
   )
   TriggerDelay_1 = UE.Trigger Delay(In=^UE.Source.OnPlay)
   ADSREnvelopeADSREnvelope_1 = ADSR Envelope.ADSR Envelope:Audio(Trigger Attack=@TriggerDelay_1.out) {also {defaults {Attack Curve=1.2, Attack Time=1, Decay Time=0, Release Time=0, Sustain Level=1}}}
   n2 = &MSP_Audio_WavePlayer_Deluxe_Stereo(Wave Asset=^1P_For_CloseRange_LP, Play Trigger=@TriggerDelay_1.out) {also {defaults {
      Is Looping?="<MetaSoundBuilderResult.SUCCEEDED: 0>"
      "Random Pitch Max (Semitones)"=0
      "Random Pitch Min (Semitones)"=0
      WavePlayer Gain dB=-3
     }}}
   n3 = &MSP_Audio_WavePlayer_Deluxe_Stereo(Wave Asset=^1P_For_CloseRange_OS, Play Trigger=&MSP_Branch(In=^UE.Source.OnPlay, Value=^isForcedCrash).False) {also {defaults {WavePlayer Gain dB=-3}}}
   AudioMixerAudioMixerStereo2_1 = "AudioMixer.Audio Mixer (Stereo, 2)"(
    In 1 R=UE.Multiply:Audio(PrimaryOperand=@ADSREnvelopeADSREnvelope_1.Out Envelope, AdditionalOperands=@n2.Out R).out
    In 1 L=UE.Multiply:Audio(PrimaryOperand=@ADSREnvelopeADSREnvelope_1.Out Envelope, AdditionalOperands=@n2.Out L).out
    In 0 R=UE.Reroute(Value=@n3.Out R).Value
    In 0 L=UE.Reroute(Value=@n3.Out L).Value
   ) {also {defaults {Gain 1=1.2}}}
   AudioMixerAudioMixerStereo2_0 = "AudioMixer.Audio Mixer (Stereo, 2)"(
    In 1 L=@AudioMixerAudioMixerMono2_0.out
    In 1 R=@AudioMixerAudioMixerMono2_0.out
    In 0 L=@AudioMixerAudioMixerStereo2_1.Out L
    In 0 R=@AudioMixerAudioMixerStereo2_1.Out R
    Gain 0=UE.Subtract:Float(AdditionalOperands=MapRange.MapRange:Float(In=^UE.Attenuation.Distance) {also {defaults {In Range B=2500}}}.Out Value) {also {defaults {PrimaryOperand=1}}}.out
   ) {also {defaults {Gain 1=0.5}}}
  ]
  links [
   {dst="@UE.OutputFormat.Stereo.Audio:0", src=AudioMixerAudioMixerStereo2_0.Out L}
   {dst="@UE.OutputFormat.Stereo.Audio:1", src=AudioMixerAudioMixerStereo2_0.Out R}
  ]
 }]
properties {
 assetImportData=&self:AssetImportData
 attenuationSettings=&/SunRose_ChargeBlast_Att
 bLooping=true
 bOverrideConcurrency=true
 duration=10000
 maxDistance=32000
 numChannels=2
 outputFormat=Stereo
 sampleRate=48000
 soundClassObject=&Fort_WeaponRanged
 volume=1.4
 concurrencyOverrides {maxCount=6, resolutionRule=StopOldest, retriggerTime=0.3, volumeScaleMode=Distance}
}
