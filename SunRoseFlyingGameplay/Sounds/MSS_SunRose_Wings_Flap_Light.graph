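# One-shot stereo MetaSound for a light SunRose wing flap in flying gameplay, played on the glider sound class.
# On play, core and broken-wing layers each play from a wave array with random variation.
# Energy drives EQ crossfades that cut the broken layer's level and lower the core's high-pass cutoff as it rises.
# The mixed layers pass through a player-affiliation mixer, and the source finishes once both players end.
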MetaSoundSource /SunRoseFlyingGameplay/Sounds/MSS_SunRose_Wings_Flap_Light
inputs [
 Energy {type=Float}
 PlayerAffiliation {type=Int32}
 Sound.BrokenWings {type=WaveAsset:Array}
 Sound.WingsCore {type=WaveAsset:Array}
 UE.Source.OnPlay {type=Trigger}
]
interfaces [UE.OutputFormat.Stereo:1.0, UE.Source.OneShot:1.0, UE.Source:1.1]
outputs [
 "UE.OutputFormat.Stereo.Audio:0" {type=Audio}
 "UE.OutputFormat.Stereo.Audio:1" {type=Audio}
 UE.Source.OneShot.OnFinished {type=Trigger}
]
pages [Default {
  input_defaults {
   Energy=1
   PlayerAffiliation=2
   UE.Source.OnPlay=null
   Sound.BrokenWings [
    /SunRoseFlyingGameplay/Sounds/wav/FN_S29_OW_Flap_Broken_01
    /SunRoseFlyingGameplay/Sounds/wav/FN_S29_OW_Flap_Broken_02
    /SunRoseFlyingGameplay/Sounds/wav/FN_S29_OW_Flap_Broken_03
    /SunRoseFlyingGameplay/Sounds/wav/FN_S29_OW_Flap_Broken_04
    /SunRoseFlyingGameplay/Sounds/wav/FN_S29_OW_Flap_Broken_05
    /SunRoseFlyingGameplay/Sounds/wav/FN_S29_OW_Flap_Broken_06
   ]
   Sound.WingsCore [
    /SunRoseFlyingGameplay/Sounds/wav/FN_S29_OW_Flap_06
    /SunRoseFlyingGameplay/Sounds/wav/FN_S29_OW_Flap_05
    /SunRoseFlyingGameplay/Sounds/wav/FN_S29_OW_Flap_04
    /SunRoseFlyingGameplay/Sounds/wav/FN_S29_OW_Flap_03
    /SunRoseFlyingGameplay/Sounds/wav/FN_S29_OW_Flap_02
    /SunRoseFlyingGameplay/Sounds/wav/FN_S29_OW_Flap_01
   ]
  }
  nodes [
   n1 = &MSP_Audio_WavePlayer_Deluxe_Stereo(Wave Asset=^Sound.BrokenWings, Play Trigger=^UE.Source.OnPlay) {also {defaults {Enable Shared Random State="<MetaSoundBuilderResult.SUCCEEDED: 0>", Max Random Decibels=1, Min Random Decibels=-1, WavePlayer Gain dB=-10}}}
   n2 = /FortMetasound/Patches/MSP_Audio_EQCrossFade(Input=^Energy, In L=@n1.Out L, In R=@n1.Out R) {also {defaults {Max.DbAdjustment=-60, Max.Value=0.5}}}
   n3 = &MSP_Audio_WavePlayer_Deluxe_Stereo(Wave Asset=^Sound.WingsCore, Play Trigger=^UE.Source.OnPlay) {also {defaults {
      Enable Shared Random State="<MetaSoundBuilderResult.SUCCEEDED: 0>"
      Max Random Decibels=1
      Min Random Decibels=-1
      "Random Pitch Max (Semitones)"=2
     }}}
   n4 = &MSP_Audio_EQCrossFade(Input=^Energy, In L=@n3.Out L, In R=@n3.Out R) {also {defaults {Max.HPF=100, Max.LPF=2500, Max.Value=0.5, Min.HPF=500, Min.LPF=2500}}}
   AudioMixerAudioMixerStereo2_0 = "AudioMixer.Audio Mixer (Stereo, 2)"(In 1 L=@n2.Out L, In 1 R=@n2.Out R, In 0 L=@n4.Out L, In 0 R=@n4.Out R)
   MSP_Audio_AffiliationMixer_Stereo_0 = /FortMetasound/Patches/MSP_Audio_AffiliationMixer_Stereo(Affiliation=^PlayerAffiliation, In L=@AudioMixerAudioMixerStereo2_0.Out L, In R=@AudioMixerAudioMixerStereo2_0.Out R, OnPlay=@n3.On Play)
   TriggerAccumulatorTriggerAccumulate2_0 = "TriggerAccumulator.Trigger Accumulate (2)"(In 0=@n3.On Finished, In 1=@n1.On Finished)
   TriggerAnyTriggerAny2_0 = "TriggerAny.Trigger Any (2)"(In 0=@n1.On Finished)
  ]
  links [
   {dst="@UE.OutputFormat.Stereo.Audio:0", src=MSP_Audio_AffiliationMixer_Stereo_0.Out L}
   {dst="@UE.OutputFormat.Stereo.Audio:1", src=MSP_Audio_AffiliationMixer_Stereo_0.Out R}
   {dst="@UE.Source.OneShot.OnFinished", src=TriggerAccumulatorTriggerAccumulate2_0.out}
  ]
 }]
properties {
 assetImportData=&self:AssetImportData
 attenuationSettings=&/SunRose_Wings_Flap_Att
 bOverrideConcurrency=true
 maxDistance=17000
 numChannels=2
 outputFormat=Stereo
 sampleRate=48000
 soundClassObject=/Game/Sounds/FortSoundClassesAndModes/Fort_Gliders
 volume=0.3
 concurrencyOverrides {bLimitToOwner=true, retriggerTime=0.05}
}
