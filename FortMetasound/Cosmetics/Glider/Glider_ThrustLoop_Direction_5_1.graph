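# MetaSound patch for a 5.1 glider thrust layer in one movement direction, mixing a random loop with threshold-triggered random oneshots.
# The smoothed Value input drives loop volume and pitch, with an added pitch offset.
# bIsBackwards, sampled on play, picks which side of the threshold fires oneshots.
# Trigger gates re-arm oneshots only after a delay and a return across the threshold, and each playback is reported through OnOneshotPlay.
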MetaSoundPatch /FortMetasound/Cosmetics/Glider/Glider_ThrustLoop_Direction_5_1
inputs [
 bIsBackwards {type=Bool}
 Interp Time {type=Time}
 Loop {type=WaveAsset:Array}
 On Play {type=Trigger}
 Oneshots {type=WaveAsset:Array}
 OneshotThreshold {type=Float}
 OneshotVolume {type=Float}
 PitchMax {type=Float}
 PitchMin {type=Float}
 PitchShift {type=Float}
 Value {type=Float}
 VolumeMax {type=Float}
 VolumeMin {type=Float}
]
interfaces [MetaSound:1.0]
outputs [
 OnOneshotPlay {type=Trigger}
 Out C {type=Audio}
 Out L {type=Audio}
 Out LF {type=Audio}
 Out R {type=Audio}
 Out SL {type=Audio}
 Out SR {type=Audio}
]
pages [Default {
  input_defaults {
   Interp Time=0.5
   On Play=null
   OneshotThreshold=0.5
   OneshotVolume=1
   PitchMax=1.2
   PitchMin=0.8
   PitchShift=0
   Value=0
   VolumeMax=1
   VolumeMin=0.2
   bIsBackwards=null
   Loop []
   Oneshots []
  }
  nodes [
   n1 = VariableAccessor.Float(Variable=VariableMutator.Float(Variable=InitVariable.Float().Variable, Value=UE.InterpTo:Audio(Interp Time=^Interp Time, Target=^Value).Value).Variable)
   FrequencyMultipliertoSemitone_0 = UE.Frequency Multiplier to Semitone:Float(Frequency Multiplier=MapRange.MapRange:Float(Out Range B=^PitchMax, Out Range A=^PitchMin, In=@n1.Value).Out Value)
   LinearGaintoDecibels_0 = UE.Linear Gain to Decibels:Float(Linear Gain=^PitchShift)
   ArrayRandomGet_1 = Array.Random Get:WaveAsset:Array(In Array=^Loop, Next=^On Play) {also {defaults {Weights []}, version=~}}
   WavePlayer_0 = UE.Wave Player:5dot1(
    Pitch Shift=UE.Add:Float(AdditionalOperands=@FrequencyMultipliertoSemitone_0.Semitones, PrimaryOperand=@LinearGaintoDecibels_0.Decibels).Out
    Play=@ArrayRandomGet_1.On Next
    Wave Asset=@ArrayRandomGet_1.Value
   ) {also {defaults {Loop="<MetaSoundBuilderResult.SUCCEEDED: 0>"}}}
   n2 = InitVariable.Trigger()
   n3 = VariableDeferredAccessor.Trigger(Variable=@n2.Variable)
   ArrayRandomGet_0 = Array.Random Get:WaveAsset:Array(In Array=^Oneshots, Next=@n3.Value) {also {defaults {Weights []}, version=~}}
   WavePlayer_1 = UE.Wave Player:5dot1(Play=@ArrayRandomGet_0.On Next, Wave Asset=@ArrayRandomGet_0.Value)
   Mixer_51_Two_0 = /FortMetasound/Patches/Mixer_51_Two(
    "2.Volume(Lin)"=^OneshotVolume
    "1.Volume(Lin)"=MapRange.MapRange:Float(Out Range B=^VolumeMax, Out Range A=^VolumeMin, In=VariableAccessor.Float(Variable=@n1.Variable).Value).Out Value
    1.InC=@WavePlayer_0.Out Front Center
    1.InL=@WavePlayer_0.Out Front Left
    1.InR=@WavePlayer_0.Out Front Right
    1.InLF=@WavePlayer_0.Out Low Frequency
    1.InSL=@WavePlayer_0.Out Side Left
    1.InSR=@WavePlayer_0.Out Side Right
    2.InC=@WavePlayer_1.Out Front Center
    2.InL=@WavePlayer_1.Out Front Left
    2.InR=@WavePlayer_1.Out Front Right
    2.InLF=@WavePlayer_1.Out Low Frequency
    2.InSL=@WavePlayer_1.Out Side Left
    2.InSR=@WavePlayer_1.Out Side Right
   )
   n4 = InitVariable.Trigger()
   TriggerDelay_0 = UE.Trigger Delay(
    Delay Time=VariableAccessor.Time(Variable=InitVariable.Time().Variable).Value
    In=VariableDeferredAccessor.Trigger(Variable=@n2.Variable).Value
   )
   n5 = TriggerCompare.Clamp:Bool(Compare=^On Play, A=^bIsBackwards)
   TriggerRouteTriggerRouteInt322_0 = "TriggerRoute.Trigger Route (Int32, 2):Int32"(Set 1=@n5.FALSE, Set 0=@n5.TRUE) {also {defaults {Value 0=2, Value 1=3}}}
   TriggerOnValueChange_0 = UE.Trigger On Value Change:Float(Value=^Value)
   n6 = VariableDeferredAccessor.Trigger(Variable=InitVariable.Trigger().Variable)
   TriggerControl_0 = UE.Trigger Control(
    Trigger In=TriggerCompare.Clamp:Float(
     B=^OneshotThreshold
     A=^Value
     Type=Convert.Enum:TriggerComparisonType:Int32(Int32=@TriggerRouteTriggerRouteInt322_0.Value)."Enum:TriggerComparisonType"
     Compare=@TriggerOnValueChange_0.Trigger
    ).TRUE
    Close=@n6.Value
    Open=VariableDeferredAccessor.Trigger(Variable=@n2.Variable).Value
   )
   TriggerAccumulatorTriggerAccumulate2_0 = "TriggerAccumulator.Trigger Accumulate (2)"(
    In 0=VariableAccessor.Trigger(Variable=VariableMutator.Trigger(Variable=@n4.Variable, Value=@TriggerDelay_0.Out).Variable).Value
    In 1=VariableAccessor.Trigger(Variable=VariableMutator.Trigger(Value=@TriggerControl_0.Trigger Out, Variable=@n6.Variable).Variable).Value
   ) {also {defaults {Auto Reset="<MetaSoundBuilderResult.SUCCEEDED: 0>"}}}
   TriggerRouteTriggerRouteInt322_1 = "TriggerRoute.Trigger Route (Int32, 2):Int32"(Set 1=@n5.FALSE, Set 0=@n5.TRUE) {also {defaults {Value 0=5, Value 1=4}}}
   TriggerOnValueChange_1 = UE.Trigger On Value Change:Float(Value=^Value)
   n7 = VariableDeferredAccessor.Trigger(Variable=@n3.Variable)
   TriggerControl_1 = UE.Trigger Control(
    Open=@TriggerAccumulatorTriggerAccumulate2_0.Out
    Trigger In=TriggerCompare.Clamp:Float(
     B=^OneshotThreshold
     A=^Value
     Type=Convert.Enum:TriggerComparisonType:Int32(Int32=@TriggerRouteTriggerRouteInt322_1.Value)."Enum:TriggerComparisonType"
     Compare=@TriggerOnValueChange_1.Trigger
    ).TRUE
    Close=@n7.Value
   )
   VariableMutator.Trigger(Value=@TriggerControl_1.Trigger Out, Variable=@n7.Variable)
  ]
  links [
   {dst="@Out C", src=Mixer_51_Two_0.OutC}
   {dst="@Out L", src=Mixer_51_Two_0.OutL}
   {dst="@Out LF", src=Mixer_51_Two_0.OutLF}
   {dst="@Out R", src=Mixer_51_Two_0.OutR}
   {dst="@Out SL", src=Mixer_51_Two_0.OutSL}
   {dst="@Out SR", src=Mixer_51_Two_0.OutSR}
   {dst="@OnOneshotPlay", src=WavePlayer_1.On Play}
  ]
 }]
