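# Looping 5.1 MetaSound for in-flight air sound on SunRose wings.
# Roll and pitch drive directional low-speed, sideways and low-energy loops, mixed with a high-speed loop through EQ crossfades keyed to speed and remaining energy.
# Ascending pins the speed value, diving nearly mutes the mix, and skimming ducks and low-passes it, before a speed-based gain and attack envelope.
# Doppler pitch shift applies only for non-local players.
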MetaSoundSource /SunRoseFlyingGameplay/Sounds/MSS_SunRose_Wings_FlyingAir_1P
inputs [
 EnergyRemaining {type=Float}
 Glider.LowSpeed {type=WaveAsset:Array}
 Glider.LowSpeedOneshots {type=WaveAsset:Array}
 Glider.SidewaysLoop {type=WaveAsset:Array}
 Glider.SidewaysOneshots {type=WaveAsset:Array}
 isAscending {type=Bool}
 isBoosting {type=Bool}
 isDiving {type=Bool}
 isDodging {type=Bool}
 IsFlapingAlpha {type=Float}
 isLocalPlayer {type=Bool}
 isSkimming {type=Bool}
 LowEnergy.Loop {type=WaveAsset:Array}
 Pitch {type=Float}
 Roll {type=Float}
 Speed {type=Float}
 SpeedThreshold.Loop {type=WaveAsset:Array}
 UE.Attenuation.Distance {type=Float}
 UE.Source.OnPlay {type=Trigger}
]
interfaces [UE.Attenuation:1.0, UE.OutputFormat.5dot1:1.0, UE.Source:1.1]
outputs [
 "UE.OutputFormat.5dot1.Audio:0" {type=Audio}
 "UE.OutputFormat.5dot1.Audio:1" {type=Audio}
 "UE.OutputFormat.5dot1.Audio:2" {type=Audio}
 "UE.OutputFormat.5dot1.Audio:3" {type=Audio}
 "UE.OutputFormat.5dot1.Audio:4" {type=Audio}
 "UE.OutputFormat.5dot1.Audio:5" {type=Audio}
]
pages [Default {
  input_defaults {
   EnergyRemaining=0
   IsFlapingAlpha=0
   Pitch=0
   Roll=0
   Speed=1047.6001
   UE.Attenuation.Distance=0
   UE.Source.OnPlay=null
   isAscending=null
   isBoosting=null
   isDiving=null
   isDodging=null
   isLocalPlayer="<MetaSoundBuilderResult.SUCCEEDED: 0>"
   isSkimming=null
   Glider.LowSpeed [/SunRoseFlyingGameplay/Sounds/wav/FN_S29_OW_LowSpeed_LP_01, /SunRoseFlyingGameplay/Sounds/wav/FN_S29_OW_LowSpeed_LP_02]
   Glider.LowSpeedOneshots []
   Glider.SidewaysLoop [/SunRoseFlyingGameplay/Sounds/wav/FN_S29_OW_AnyDirection_lp_01]
   Glider.SidewaysOneshots []
   LowEnergy.Loop [/SunRoseFlyingGameplay/Sounds/wav/FN_S29_OW_Low_Energy_lp_02, /SunRoseFlyingGameplay/Sounds/wav/FN_S29_OW_Low_Energy_lp_01]
   SpeedThreshold.Loop [/SunRoseFlyingGameplay/Sounds/wav/FN_S29_OW_HighSpeed_lp_02, /SunRoseFlyingGameplay/Sounds/wav/FN_S29_OW_HighSpeed_lp_01]
  }
  nodes [
   InitVariable.Float()
   n1 = &MSP_Branch(In=^UE.Source.OnPlay, Value=^isLocalPlayer)
   n2 = UE.Reroute(Value=^Speed)
   TriggerRouteTriggerRouteFloat2_0 = "TriggerRoute.Trigger Route (Float, 2):Float"(
    Set 1=@n1.FALSE
    Set 0=@n1.TRUE
    Value 1=UE.Reroute(Value=/FortMetasound/Patches/MSP_Doppler(Distance=^UE.Attenuation.Distance, Speed=@n2.Value) {also {defaults {StrengthMultiplier=3}}}.SemitoneModifier).Value
   )
   n3 = VariableAccessor.Float(Variable=VariableMutator.Float(Variable=InitVariable.Float().Variable, Value=@TriggerRouteTriggerRouteFloat2_0.Value).Variable)
   n4 = VariableAccessor.Float(Variable=@n3.Variable)
   n5 = VariableAccessor.Float(Variable=@n4.Variable)
   SemitonetoFrequencyMultiplier_0 = UE.Semitone to Frequency Multiplier:Float(Semitones=VariableAccessor.Float(Variable=@n5.Variable).Value)
   n6 = MapRange.MapRange:Float(In=AbsoluteValue.Abs:Float(Input=^Roll).Value) {also {defaults {In Range B=3}}}
   TriggerOnValueChange_4 = UE.Trigger On Value Change:Float(Value=^Roll)
   n7 = VariableDeferredAccessor.Float(Variable=InitVariable.Float().Variable)
   n8 = TriggerCompare.Clamp:Float(B=@n6.Out Value, Compare=@TriggerOnValueChange_4.Trigger, A=@n7.Value) {also {defaults {Type=3}}}
   TriggerRouteTriggerRouteFloat2_5 = "TriggerRoute.Trigger Route (Float, 2):Float"(Set 1=@n8.FALSE, Set 0=@n8.TRUE) {also {defaults {Value 0=0.75, Value 1=0.1}}}
   n9 = UE.ConversionFloatToTime(In=@TriggerRouteTriggerRouteFloat2_5.Value)
   n10 = VariableAccessor.Float(Variable=VariableMutator.Float(Value=UE.InterpTo:Audio(Interp Time=@n9.out, Target=@n6.Out Value).Value, Variable=@n7.Variable).Variable)
   n11 = /FortMetasound/Cosmetics/Glider/Glider_ThrustLoop_Direction_5_1(
    loop=^Glider.SidewaysLoop
    Oneshots=^Glider.SidewaysOneshots
    On Play=^UE.Source.OnPlay
    PitchShift=@SemitonetoFrequencyMultiplier_0.Frequency Multiplier
    Value=@n10.Value
   ) {also {defaults {Interp Time=0, OneshotThreshold=1, OneshotVolume=0.2, VolumeMin=0}}}
   SemitonetoFrequencyMultiplier_3 = UE.Semitone to Frequency Multiplier:Float(Semitones=@n5.Value)
   n12 = VariableAccessor.Float(Variable=@n10.Variable)
   TriggerOnValueChange_5 = UE.Trigger On Value Change:Float(Value=@n12.Value)
   TriggerOnValueChange_6 = UE.Trigger On Value Change:Float(Value=^Pitch)
   TriggerAnyTriggerAny3_0 = "TriggerAny.Trigger Any (3)"(In 0=^UE.Source.OnPlay, In 2=@TriggerOnValueChange_5.Trigger, In 1=@TriggerOnValueChange_6.Trigger)
   n13 = VariableAccessor.Float(Variable=VariableMutator.Float(Value=^Pitch, Variable=InitVariable.Float().Variable).Variable)
   n14 = TriggerCompare.Clamp:Float(
    B=UE.Reroute(Value=UE.Reroute(Value=@n12.Value).Value).Value
    Compare=@TriggerAnyTriggerAny3_0.out
    A=VariableAccessor.Float(Variable=@n13.Variable).Value
   ) {also {defaults {Type=3}}}
   TriggerRouteTriggerRouteFloat2_3 = "TriggerRoute.Trigger Route (Float, 2):Float"(Value 1=UE.Reroute(Value=UE.Reroute(Value=@n12.Value).Value).Value, Set 1=@n14.FALSE, Set 0=@n14.TRUE, Value 0=@n13.Value)
   n15 = VariableAccessor.Float(Variable=VariableMutator.Float(Variable=InitVariable.Float().Variable, Value=@TriggerRouteTriggerRouteFloat2_3.Value).Variable)
   n16 = &Glider_ThrustLoop_Direction_5_1(
    loop=^Glider.LowSpeed
    Oneshots=^Glider.LowSpeedOneshots
    On Play=^UE.Source.OnPlay
    PitchShift=@SemitonetoFrequencyMultiplier_3.Frequency Multiplier
    Value=@n15.Value
   ) {also {defaults {Interp Time=0.35, PitchMin=1, VolumeMax=2, VolumeMin=1.5, bIsBackwards="<MetaSoundBuilderResult.SUCCEEDED: 0>"}}}
   TriggerOnValueChange_3 = UE.Trigger On Value Change:Float(Value=^Speed)
   TriggerOnValueChange_0 = UE.Trigger On Value Change:Bool(Value=^isAscending)
   TriggerAnyTriggerAny2_2 = "TriggerAny.Trigger Any (2)"(In 0=^UE.Source.OnPlay, In 1=@TriggerOnValueChange_0.Trigger)
   n17 = TriggerCompare.Clamp:Bool(A=UE.Reroute(Value=^isAscending).Value, Compare=@TriggerAnyTriggerAny2_2.out) {also {defaults {B="<MetaSoundBuilderResult.SUCCEEDED: 0>"}}}
   TriggerRouteTriggerRouteFloat2_2 = "TriggerRoute.Trigger Route (Float, 2):Float"(Value 1=Clamp.Clamp:Float(In=^Speed) {also {defaults {Max=1800}}}.Value, Set 1=@n17.FALSE, Set 0=@n17.TRUE) {also {defaults {Value 0=950}}}
   n18 = VariableAccessor.Float(Variable=VariableMutator.Float(
     Variable=InitVariable.Float().Variable
     Value=/FortMetasound/Patches/MSP_ClampFloatChange(OnPlay=^UE.Source.OnPlay, ForceCalculation=@TriggerOnValueChange_3.Trigger, Value=@TriggerRouteTriggerRouteFloat2_2.Value) {also {defaults {Tolerance=50}}}.Output
    ).Variable)
   n19 = VariableAccessor.Float(Variable=@n18.Variable)
   n20 = /FortMetasound/Patches/MSP_Audio_EQCrossFade_51(In C=@n16.Out C, In L=@n16.Out L, In LF=@n16.Out LF, In R=@n16.Out R, In SL=@n16.Out SL, In SR=@n16.Out SR, Input=@n19.Value) {also {defaults {Interp Time=0.1, Max.Value=950, Min.DbAdjustment=-65, Min.Value=1150}}}
   n21 = /FortMetasound/Patches/Mixer_51_Two(
    2.InC=@n11.Out C
    2.InL=@n11.Out L
    2.InLF=@n11.Out LF
    2.InR=@n11.Out R
    2.InSL=@n11.Out SL
    2.InSR=@n11.Out SR
    1.InC=@n20.Out C
    1.InL=@n20.Out L
    1.InLF=@n20.Out LF
    1.InR=@n20.Out R
    1.InSL=@n20.Out SL
    1.InSR=@n20.Out SR
   )
   n22 = &MSP_Audio_EQCrossFade_51(Input=^EnergyRemaining, In C=@n21.OutC, In L=@n21.OutL, In LF=@n21.OutLF, In R=@n21.OutR, In SL=@n21.OutSL, In SR=@n21.OutSR) {also {defaults {Max.Value=0.4}}}
   ArrayRandomGet_0 = Array.Random Get:WaveAsset:Array(In Array=^SpeedThreshold.Loop, Next=^UE.Source.OnPlay) {also {defaults {Weights []}}}
   WavePlayer_0 = UE.Wave Player:5dot1(Play=@ArrayRandomGet_0.On Next, Wave Asset=@ArrayRandomGet_0.Value, Pitch Shift=@n3.Value) {also {defaults {Loop="<MetaSoundBuilderResult.SUCCEEDED: 0>"}}}
   n23 = &MSP_Audio_EQCrossFade_51(
    Input=@n18.Value
    In C=@WavePlayer_0.Out Front Center
    In L=@WavePlayer_0.Out Front Left
    In R=@WavePlayer_0.Out Front Right
    In LF=@WavePlayer_0.Out Low Frequency
    In SL=@WavePlayer_0.Out Side Left
    In SR=@WavePlayer_0.Out Side Right
   ) {also {defaults {Interp Time=0.1, Max.DbAdjustment=1, Max.Value=1300, Min.DbAdjustment=-65, Min.Value=925}}}
   n24 = &MSP_Audio_EQCrossFade_51(Input=^EnergyRemaining, In C=@n23.Out C, In L=@n23.Out L, In LF=@n23.Out LF, In R=@n23.Out R, In SL=@n23.Out SL, In SR=@n23.Out SR) {also {defaults {Max.DbAdjustment=1, Max.Value=0.4, Min.DbAdjustment=-3, Min.HPF=550, Min.Value=0.15}}}
   n25 = /FortMetasound/SoundCue/MSP_ContinuousModulator(Value In=^EnergyRemaining) {also {defaults {Input.Max=0.35, Output.Max=0, Output.Min=5}}}
   SemitonetoFrequencyMultiplier_1 = UE.Semitone to Frequency Multiplier:Float(Semitones=UE.Reroute(Value=@n25.Value Out).Value)
   SemitonetoFrequencyMultiplier_2 = UE.Semitone to Frequency Multiplier:Float(Semitones=@n4.Value)
   SemitonetoFrequencyMultiplier_4 = UE.Semitone to Frequency Multiplier:Float(Semitones=UE.Add:Float(PrimaryOperand=@n25.Value Out) {also {defaults {AdditionalOperands=5}}}.out)
   n26 = &Glider_ThrustLoop_Direction_5_1(
    loop=^LowEnergy.Loop
    On Play=^UE.Source.OnPlay
    PitchMin=@SemitonetoFrequencyMultiplier_1.Frequency Multiplier
    PitchShift=@SemitonetoFrequencyMultiplier_2.Frequency Multiplier
    PitchMax=@SemitonetoFrequencyMultiplier_4.Frequency Multiplier
    Value=VariableAccessor.Float(Variable=@n15.Variable).Value
   ) {also {defaults {VolumeMax=1.3, VolumeMin=1, bIsBackwards="<MetaSoundBuilderResult.SUCCEEDED: 0>"}}}
   n27 = &MSP_Audio_EQCrossFade_51(Input=^EnergyRemaining, In C=@n26.Out C, In L=@n26.Out L, In LF=@n26.Out LF, In R=@n26.Out R, In SL=@n26.Out SL, In SR=@n26.Out SR) {also {defaults {Max.DbAdjustment=-60, Max.HPF=600, Max.Value=0.65, Min.DbAdjustment=6, Min.Value=0.3}}}
   n28 = &Mixer_51_Two(
    1.InC=@n24.Out C
    1.InL=@n24.Out L
    1.InLF=@n24.Out LF
    1.InR=@n24.Out R
    1.InSL=@n24.Out SL
    1.InSR=@n24.Out SR
    2.InC=@n27.Out C
    2.InL=@n27.Out L
    2.InLF=@n27.Out LF
    2.InR=@n27.Out R
    2.InSL=@n27.Out SL
    2.InSR=@n27.Out SR
   ) {also {defaults {"2.Volume(Lin)"=1.5}}}
   n29 = &Mixer_51_Two(
    1.InC=@n22.Out C
    1.InL=@n22.Out L
    1.InLF=@n22.Out LF
    1.InR=@n22.Out R
    1.InSL=@n22.Out SL
    1.InSR=@n22.Out SR
    2.InC=@n28.OutC
    2.InL=@n28.OutL
    2.InLF=@n28.OutLF
    2.InR=@n28.OutR
    2.InSL=@n28.OutSL
    2.InSR=@n28.OutSR
   )
   n30 = InitVariable.Float()
   TriggerOnValueChange_1 = UE.Trigger On Value Change:Bool(Value=^isDiving)
   TriggerAnyTriggerAny2_1 = "TriggerAny.Trigger Any (2)"(In 0=^UE.Source.OnPlay, In 1=@TriggerOnValueChange_1.Trigger)
   n31 = TriggerCompare.Clamp:Bool(A=^isDiving, Compare=@TriggerAnyTriggerAny2_1.out) {also {defaults {B="<MetaSoundBuilderResult.SUCCEEDED: 0>"}}}
   TriggerRouteTriggerRouteFloat2_1 = "TriggerRoute.Trigger Route (Float, 2):Float"(Set 1=@n31.FALSE, Set 0=@n31.TRUE) {also {defaults {Value 0=1}}}
   n32 = &MSP_Audio_EQCrossFade_51(
    In C=@n29.OutC
    In L=@n29.OutL
    In LF=@n29.OutLF
    In R=@n29.OutR
    In SL=@n29.OutSL
    In SR=@n29.OutSR
    Input=VariableAccessor.Float(Variable=VariableMutator.Float(Variable=@n30.Variable, Value=@TriggerRouteTriggerRouteFloat2_1.Value).Variable).Value
   ) {also {defaults {Max.DbAdjustment=-60}}}
   TriggerOnValueChange_2 = UE.Trigger On Value Change:Bool(Value=^isSkimming)
   TriggerAnyTriggerAny2_0 = "TriggerAny.Trigger Any (2)"(In 0=^UE.Source.OnPlay, In 1=@TriggerOnValueChange_2.Trigger)
   n33 = &MSP_Branch(Value=UE.Reroute(Value=^isSkimming).Value, In=@TriggerAnyTriggerAny2_0.out)
   TriggerRouteTriggerRouteFloat2_4 = "TriggerRoute.Trigger Route (Float, 2):Float"(Set 0=@n33.FALSE, Set 1=@n33.TRUE) {also {defaults {Value 0=0, Value 1=1}}}
   n34 = &MSP_Audio_EQCrossFade_51(
    In C=@n32.Out C
    In L=@n32.Out L
    In LF=@n32.Out LF
    In R=@n32.Out R
    In SL=@n32.Out SL
    In SR=@n32.Out SR
    Input=@TriggerRouteTriggerRouteFloat2_4.Value
   ) {also {defaults {Interp Time=1, Max.DbAdjustment=-10, Max.LPF=2000}}}
   n35 = &MSP_Audio_EQCrossFade_51(
    In C=@n34.Out C
    In L=@n34.Out L
    In LF=@n34.Out LF
    In R=@n34.Out R
    In SL=@n34.Out SL
    In SR=@n34.Out SR
    Input=VariableAccessor.Float(Variable=@n19.Variable).Value
   ) {also {defaults {Interp Time=1, Max.DbAdjustment=2, Max.Value=1800, Min.Value=900}}}
   ADSREnvelopeADSREnvelope_0 = ADSR Envelope.ADSR Envelope:Audio(Trigger Attack=^UE.Source.OnPlay) {also {defaults {Attack Curve=2, Attack Time=0.7, Decay Time=0, Release Time=3, Sustain Level=1}}}
   n36 = VariableAccessor.Audio(Variable=VariableMutator.Audio(Value=@ADSREnvelopeADSREnvelope_0.Out Envelope, Variable=InitVariable.Audio().Variable).Variable)
   Multiply_0 = UE.Multiply:Audio(PrimaryOperand=@n35.Out LF, AdditionalOperands=@n36.Value)
   Multiply_1 = UE.Multiply:Audio(PrimaryOperand=@n35.Out L, AdditionalOperands=@n36.Value)
   Multiply_2 = UE.Multiply:Audio(PrimaryOperand=@n35.Out R, AdditionalOperands=@n36.Value)
   Multiply_3 = UE.Multiply:Audio(PrimaryOperand=@n35.Out SR, AdditionalOperands=@n36.Value)
   Multiply_4 = UE.Multiply:Audio(PrimaryOperand=@n35.Out C, AdditionalOperands=@n36.Value)
   Multiply_5 = UE.Multiply:Audio(PrimaryOperand=@n35.Out SL, AdditionalOperands=@n36.Value)
  ]
  links [
   {dst="@UE.OutputFormat.5dot1.Audio:3", src=Multiply_0.out}
   {dst="@UE.OutputFormat.5dot1.Audio:0", src=Multiply_1.out}
   {dst="@UE.OutputFormat.5dot1.Audio:1", src=Multiply_2.out}
   {dst="@UE.OutputFormat.5dot1.Audio:5", src=Multiply_3.out}
   {dst="@UE.OutputFormat.5dot1.Audio:2", src=Multiply_4.out}
   {dst="@UE.OutputFormat.5dot1.Audio:4", src=Multiply_5.out}
  ]
 }]
properties {
 assetImportData=&self:AssetImportData
 attenuationSettings=&/SunRose_Wings_InFlightAir_Att
 bLooping=true
 duration=10000
 maxDistance=3650
 numChannels=6
 outputFormat=FiveDotOne
 sampleRate=48000
 soundClassObject=&Fort_Gliders_InAir
 volume=0.2
}
