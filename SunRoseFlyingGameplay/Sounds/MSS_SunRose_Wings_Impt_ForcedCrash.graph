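# One-shot stereo MetaSound source for the SunRose wings forced-crash impact sound, classed under gliders.
# On play, a deluxe wave player patch plays from an array of two impact waves with slight random decibel variation.
# Its output feeds an affiliation mixer patch driven by the PlayerAffiliation input, and playback completion fires OnFinished.
# It uses the wings skim attenuation and a concurrency override limited to owner.
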MetaSoundSource /SunRoseFlyingGameplay/Sounds/MSS_SunRose_Wings_Impt_ForcedCrash
inputs [
 PlayerAffiliation {type=Int32}
 Sound.Oneshot {type=WaveAsset:Array}
 UE.Source.OnPlay {type=Trigger}
]
interfaces [UE.OutputFormat.Stereo:1.0, UE.Source.OneShot:1.0, UE.Source:1.1]
outputs [
 "UE.OutputFormat.Stereo.Audio:0" {type=Audio}
 "UE.OutputFormat.Stereo.Audio:1" {type=Audio}
 UE.Source.OneShot.OnFinished {type=Trigger}
]
pages [Default {
  input_defaults {
   PlayerAffiliation=2
   UE.Source.OnPlay=null
   Sound.Oneshot [/SunRoseFlyingGameplay/Sounds/wav/FN_S29_OW_ForceCrash_Impt_01, /SunRoseFlyingGameplay/Sounds/wav/FN_S29_OW_ForceCrash_Impt_02]
  }
  nodes [
   MSP_Audio_WavePlayer_Deluxe_Stereo_0 = &MSP_Audio_WavePlayer_Deluxe_Stereo(Wave Asset=^Sound.Oneshot, Play Trigger=^UE.Source.OnPlay) {also {defaults {Enable Shared Random State="<MetaSoundBuilderResult.SUCCEEDED: 0>", Max Random Decibels=1, Min Random Decibels=-1}}}
   MSP_Audio_AffiliationMixer_Stereo_0 = /FortMetasound/Patches/MSP_Audio_AffiliationMixer_Stereo(
    Affiliation=^PlayerAffiliation
    OnPlay=@MSP_Audio_WavePlayer_Deluxe_Stereo_0.On Play
    In L=@MSP_Audio_WavePlayer_Deluxe_Stereo_0.Out L
    In R=@MSP_Audio_WavePlayer_Deluxe_Stereo_0.Out R
   )
  ]
  links [
   {dst="@UE.OutputFormat.Stereo.Audio:0", src=MSP_Audio_AffiliationMixer_Stereo_0.Out L}
   {dst="@UE.OutputFormat.Stereo.Audio:1", src=MSP_Audio_AffiliationMixer_Stereo_0.Out R}
   {dst="@UE.Source.OneShot.OnFinished", src=MSP_Audio_WavePlayer_Deluxe_Stereo_0.On Finished}
  ]
 }]
properties {
 assetImportData=&self:AssetImportData
 attenuationSettings=&/SunRose_Wings_Skim_Att
 bOverrideConcurrency=true
 maxDistance=5300
 numChannels=2
 outputFormat=Stereo
 sampleRate=48000
 soundClassObject=/Game/Sounds/FortSoundClassesAndModes/Fort_Gliders
 volume=0.2
 concurrencyOverrides {bLimitToOwner=true, retriggerTime=0.05}
}
