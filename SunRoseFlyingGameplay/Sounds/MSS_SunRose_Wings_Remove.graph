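# MetaSound Source that plays the one-shot stereo sound for removing wings in the SunRose flying gameplay.
# On play, a random get picks from the "Wings Off Sound" wave array, which holds one default asset, and triggers a stereo Wave Player.
# The player feeds both output channels and ends the source when playback finishes.
# It uses its own attenuation, the melee weapon sound class and owner-limited concurrency.
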MetaSoundSource /SunRoseFlyingGameplay/Sounds/MSS_SunRose_Wings_Remove
inputs [
 Sound.Wings Off Sound {type=WaveAsset:Array}
 UE.Source.OnPlay {type=Trigger}
]
interfaces [UE.OutputFormat.Stereo:1.0, UE.Source.OneShot:1.0, UE.Source:1.1]
outputs [
 "UE.OutputFormat.Stereo.Audio:0" {type=Audio}
 "UE.OutputFormat.Stereo.Audio:1" {type=Audio}
 UE.Source.OneShot.OnFinished {type=Trigger}
]
pages [Default {
  input_defaults {UE.Source.OnPlay=null, Sound.Wings Off Sound [/SunRoseFlyingGameplay/Sounds/wav/S29_FN_OW_WIngs_Removed]}
  nodes [
   ArrayRandomGet_0 = Array.Random Get:WaveAsset:Array(In Array=^Sound.Wings Off Sound, Next=^UE.Source.OnPlay)
   WavePlayer_0 = UE.Wave Player:Stereo(Play=@ArrayRandomGet_0.On Next, Wave Asset=@ArrayRandomGet_0.Value)
  ]
  links [
   {dst="@UE.Source.OneShot.OnFinished", src=WavePlayer_0.On Finished}
   {dst="@UE.OutputFormat.Stereo.Audio:0", src=WavePlayer_0.Out Left}
   {dst="@UE.OutputFormat.Stereo.Audio:1", src=WavePlayer_0.Out Right}
  ]
 }]
properties {
 assetImportData=&self:AssetImportData
 attenuationSettings=&/SunRose_Wings_Remove_Att
 bOverrideConcurrency=true
 maxDistance=24300
 numChannels=2
 outputFormat=Stereo
 sampleRate=48000
 soundClassObject=&Fort_WeaponMelee
 volume=0.5
 concurrencyOverrides {bLimitToOwner=true, maxCount=6, retriggerTime=0.05, volumeScale=0.5}
}
