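; Sound library animation context for SunRose flying gameplay that feeds gameplay state into the audio components it plays.
; Play calls the parent, then loops over the returned components, running two helpers gated by AudioParams flags.
; PassNormalizedEnergy sends the wings' remaining energy percentage as the "Energy" float.
; PassPlayerAffiliation finds the player pawn from the owner or its owner, deriving a "PlayerAffiliation" integer from local view and team affiliation.
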(fn PassNormalizedEnergy
  (params (Owner :type Actor%20Object%20Reference) (AudioComponent :type Audio%20Component%20Object%20Reference))
  (Utilities|IsValid AudioComponent
    (:Is%20Valid
      (bind _asb_athena_sun_rose_wings (Utilities|Casting|CastToB_Athena_SunRose_Wings (Variables|GetOwner)))
      (Audio|Parameter|SetFloatParameter (Variables|GetAudioComponent) "Energy" (Variables|/SunRoseFlyingGameplay/Items/Wings/Blueprints/B_Athena_SunRose_Wings.B_Athena_SunRose_Wings_C|GetPctEnergyRemaining _asb_athena_sun_rose_wings) :!class /Script/AudioExtensions.AudioParameterControllerInterface))))

(fn PassPlayerAffiliation
  (params (Owner :type Actor%20Object%20Reference) (AudioComponent :type Audio%20Component%20Object%20Reference))
  (bind _owner (Variables|GetOwner))
  (Utilities|IsValid AudioComponent
    (:Is%20Valid
      (bind _asplayer_pawn_athena (Utilities|Casting|CastToPlayerPawn_Athena _owner)
        (:then
          (LocalVariables|SetPlayerPawn _asplayer_pawn_athena))
        (:CastFailed
          (bind _asplayer_pawn_athena_2 (Utilities|Casting|CastToPlayerPawn_Athena (Actor|GetOwner _owner)))
          (LocalVariables|SetPlayerPawn _asplayer_pawn_athena_2)))
      (if (Pawn|IsLocallyViewed (LocalVariables|GetPlayerPawn))
        (LocalVariables|SetPlayerAffiliation 0)
        (else
          (switch Utilities|FlowControl|Switch|SwitchonEFortTeamAffiliation (Team|GetTeamAffiliationforActortoLocalPlayer (LocalVariables|GetPlayerPawn))
            (:Friendly
              _)
            (:Neutral
              _)
            (:Hostile
              (LocalVariables|SetPlayerAffiliation 2)
              (bind _audiocomponent (Variables|GetAudioComponent)
                (:then
                  (Audio|Parameter|SetIntegerParameter _audiocomponent "PlayerAffiliation" (LocalVariables|GetPlayerAffiliation) :!class /Script/AudioExtensions.AudioParameterControllerInterface))
                (:else))))
          (LocalVariables|SetPlayerAffiliation 1))))))

(fn Play
  (params (InEventData :type Sound%20Library%20Context%20Event%20Input%20Structure%20%28by%20ref%29))
  (bind _owner (Utilities|Struct|BreakSoundLibraryContextEventInput InEventData))
  (bind (_outcomponents _returnvalue) (|CallParentFunction InEventData))
  (for _array_element _outcomponents
    (if (Variables|AudioParams|GetIncludeEnergyLevel)
      (AudioParams|PassNormalizedEnergy _owner _array_element)
      (else
        _))
    (if (Variables|AudioParams|GetIncludeAffiliationIntParam?)
      (AudioParams|PassPlayerAffiliation _owner _array_element)))
  (return _outcomponents _returnvalue))
