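# Sound cue that plays the boost-start sound for SunRose wings in the flying gameplay.
# A PlayerAffiliation switch sends all its branches to one modulator, so every affiliation sounds alike.
# The modulator slightly varies pitch and volume over an equal-weight random pick of four boost waves.
# It uses the in-air glider sound class, ChargeBlast tell attenuation, and overridden concurrency that prevents new plays at the limit.
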SoundCue /SunRoseFlyingGameplay/Sounds/SunRose_Wings_Boost_Cue
root=Switch_0
nodes [
 {id=Modulator_0, children [Random_1], properties {pitchMax=1, volumeMax=1.1, volumeMin=1}}
 {id=Random_1, children [WavePlayer_0,WavePlayer_1,WavePlayer_2,WavePlayer_3], properties {weights [1,1,1,1]}}
 {id=Switch_0, children [Modulator_0,Modulator_0,Modulator_0,Modulator_0], properties {intParameterName=PlayerAffiliation}}
 {id=WavePlayer_0, properties {soundWaveAssetPtr=/SunRoseFlyingGameplay/Sounds/wav/FN_S29_OW_Boost_Start_01}}
 {id=WavePlayer_1, properties {soundWaveAssetPtr=/SunRoseFlyingGameplay/Sounds/wav/FN_S29_OW_Boost_Start_02}}
 {id=WavePlayer_2, properties {soundWaveAssetPtr=/SunRoseFlyingGameplay/Sounds/wav/FN_S29_OW_Boost_Start_03}}
 {id=WavePlayer_3, properties {soundWaveAssetPtr=/SunRoseFlyingGameplay/Sounds/wav/FN_S29_OW_Boost_Start_04}}
]
properties {
 attenuationSettings=&/SunRose_ChargeBlast_Tell_Att
 bOverrideConcurrency=true
 maxDistance=18000
 soundClassObject=&Fort_Gliders_InAir
 volumeMultiplier=0.65
 attenuationOverrides {distanceAlgorithm=NaturalSound, falloffDistance=3000, attenuationShapeExtents {x=1000}}
 concurrencyOverrides {resolutionRule=PreventNew, retriggerTime=1}
 concurrencySet [&Player_Melee_Weapon_Conc]
}
