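# SunRose Wings crash impact SoundCue, layering close and distant impact sounds through separate attenuation assets.
# The close layer mixes random impacts with quieter, briefly delayed dirt-and-stone debris; the distant layer picks random distant impacts.
# A Water branch adds randomized splash sounds over that full mix when true, while all four PlayerAffiliation switch inputs share this same branch.
# Per-owner concurrency stops the oldest instance when the limit is reached.
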SoundCue /SunRoseFlyingGameplay/Sounds/SunRose_Wings_Crash_Impact_Cue
root=Switch_0
nodes [
 {
  id=Attenuation_0
  children [Modulator_1]
  properties {attenuationSettings=&/SunRose_ChargeBlast_Impact_Distant_Att}
 }
 {id=Attenuation_2, children [Mixer_1], properties {attenuationSettings=&/SunRose_ChargeBlast_Impact_Close_Att}}
 {
  id=Attenuation_3
  children [Modulator_0]
  properties {attenuationSettings=&SunRose_ChargeBlast_Impact_Close_Att}
 }
 {id=Branch_0, children [Mixer_2,Mixer_0,Mixer_0], properties {boolParameterName=Water}}
 {id=Delay_0, children [Random_3], properties {delayMax=0.2, delayMin=0.2}}
 {id=Mixer_0, children [Attenuation_2, Attenuation_0], properties {inputVolume [1,1]}}
 {id=Mixer_1, children [Modulator_3,Modulator_2], properties {inputVolume [0.5,0.5]}}
 {id=Mixer_2, children [Attenuation_3, Mixer_0], properties {inputVolume [1,1]}}
 {id=Modulator_0, children [Random_1], properties {pitchMin=0.9, volumeMax=0.75, volumeMin=0.6}}
 {id=Modulator_1, children [Random_0]}
 {id=Modulator_2, children [Delay_0], properties {volumeMax=0.3, volumeMin=0.2}}
 {id=Modulator_3, children [Random_2], properties {volumeMax=1.1, volumeMin=1}}
 {id=Random_0, children [WavePlayer_2, WavePlayer_3, WavePlayer_14], properties {weights [1,1,1]}}
 {id=Random_1, children [WavePlayer_0,WavePlayer_5], properties {weights [1,1]}}
 {id=Random_2, children [WavePlayer_6, WavePlayer_7, WavePlayer_8, WavePlayer_9, WavePlayer_13], properties {weights [1,1,1,1,1]}}
 {id=Random_3, children [WavePlayer_1, WavePlayer_4, WavePlayer_10, WavePlayer_11, WavePlayer_12], properties {weights [1,1,1,1,1]}}
 {id=Switch_0, children [Branch_0,Branch_0,Branch_0,Branch_0], properties {intParameterName=PlayerAffiliation}}
 {id=WavePlayer_0, properties {soundWaveAssetPtr=/FNGameplayCues/Athena/Sounds/Vehicles/Ostrich/Veh_Ostrich_FS_Water_Land_Hv_01}}
 {id=WavePlayer_1, properties {soundWaveAssetPtr=/Game/Athena/Sounds/Weapons/Explosions/EXPL_Debris_Dirt_Stone_01}}
 {id=WavePlayer_10, properties {soundWaveAssetPtr=/Game/Athena/Sounds/Weapons/Explosions/EXPL_Debris_Dirt_Stone_05}}
 {id=WavePlayer_11, properties {soundWaveAssetPtr=/Game/Athena/Sounds/Weapons/Explosions/EXPL_Debris_Dirt_Stone_02}}
 {id=WavePlayer_12, properties {soundWaveAssetPtr=/Game/Athena/Sounds/Weapons/Explosions/EXPL_Debris_Dirt_Stone_04}}
 {id=WavePlayer_13, properties {soundWaveAssetPtr=/SunRoseFlyingGameplay/Sounds/wav/FN_S29_OW_Impt_05}}
 {id=WavePlayer_14, properties {soundWaveAssetPtr=/SunRoseFlyingGameplay/Sounds/wav/FN_S29_OW_Impt_Dist_01}}
 {id=WavePlayer_2, properties {soundWaveAssetPtr=/SunRoseFlyingGameplay/Sounds/wav/FN_S29_OW_Impt_Dist_03}}
 {id=WavePlayer_3, properties {soundWaveAssetPtr=/SunRoseFlyingGameplay/Sounds/wav/FN_S29_OW_Impt_Dist_02}}
 {id=WavePlayer_4, properties {soundWaveAssetPtr=/Game/Athena/Sounds/Weapons/Explosions/EXPL_Debris_Dirt_Stone_03}}
 {id=WavePlayer_5, properties {soundWaveAssetPtr=/FNGameplayCues/Athena/Sounds/Vehicles/Ostrich/Veh_Ostrich_FS_Water_Land_Hv_02}}
 {id=WavePlayer_6, properties {soundWaveAssetPtr=/SunRoseFlyingGameplay/Sounds/wav/FN_S29_OW_Impt_01}}
 {id=WavePlayer_7, properties {soundWaveAssetPtr=/SunRoseFlyingGameplay/Sounds/wav/FN_S29_OW_Impt_02}}
 {id=WavePlayer_8, properties {soundWaveAssetPtr=/SunRoseFlyingGameplay/Sounds/wav/FN_S29_OW_Impt_03}}
 {id=WavePlayer_9, properties {soundWaveAssetPtr=/SunRoseFlyingGameplay/Sounds/wav/FN_S29_OW_Impt_04}}
]
properties {
 bOverrideConcurrency=true
 maxDistance=26000
 soundClassObject=&Fort_Gadgets
 volumeMultiplier=1.75
 attenuationOverrides {
  bAttenuateWithLPF=true
  bEnableOcclusion=true
  falloffDistance=13000
  lPFFrequencyAtMax=1000
  lPFFrequencyAtMin=10000
  lPFRadiusMax=2000
  lPFRadiusMin=350
  nonSpatializedRadiusMode=Direct2D
  occlusionLowPassFilterFrequency=2000
  occlusionTraceChannel=ECC_GameTraceChannel6
  occlusionVolumeAttenuation=0.8
  stereoSpread=380
  attenuationShapeExtents {x=300}
 }
 concurrencyOverrides {bLimitToOwner=true, maxCount=3, resolutionRule=StopOldest}
}
