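# Sound cue for the SunRose wings crash tell, playing a random ChargeBlast tell wave in the in-air glider sound class.
# A forced crash plays silence; otherwise an islocal branch picks the local or remote chain.
# Both randomize the same six waves through shared attenuation, with the local chain quieter and shaped by a pitch envelope.
# A concurrency override allows one instance per owner, blocking new plays.
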SoundCue /SunRoseFlyingGameplay/Sounds/SunRose_Wings_Crash_Tell_Cue
root=Branch_1
nodes [
 {
  id=Attenuation_118
  children [Random_1]
  properties {attenuationSettings=&/SunRose_ChargeBlast_Tell_Att, attenuationOverrides {stereoSpread=29871.3984}}
 }
 {id=Attenuation_119, children [Random_2], properties {attenuationSettings=&SunRose_ChargeBlast_Tell_Att}}
 {id=Branch_0, children [Modulator_2,Modulator_3,Modulator_3], properties {boolParameterName=islocal}}
 {id=Branch_1, children [null,Branch_0,Branch_0], properties {boolParameterName=isForcedCrash}}
 {
  id=Enveloper_0
  children [Modulator_0]
  properties {pitchCurve {editorCurveData {keys [<at -0.0006 0.0041>, {time=0}, <at 0.5 2>]}}, volumeCurve {editorCurveData {keys [<at 0 1>]}}}
 }
 {id=Modulator_0, children [Attenuation_118], properties {volumeMax=0.5, volumeMin=0.5}}
 {id=Modulator_2, children [Enveloper_0], properties {pitchMax=1, pitchMin=1, volumeMax=0.8, volumeMin=0.8}}
 {id=Modulator_3, children [Attenuation_119], properties {pitchMax=1, pitchMin=1, volumeMax=1.2, volumeMin=1.2}}
 {
  id=Random_1
  children [WavePlayer_8, WavePlayer_9, WavePlayer_10, WavePlayer_11, WavePlayer_12, WavePlayer_13]
  properties {weights [1,1,1,1,1,1]}
 }
 {
  id=Random_2
  children [WavePlayer_16, WavePlayer_17, WavePlayer_18, WavePlayer_19, WavePlayer_20, WavePlayer_21]
  properties {weights [1,1,1,1,1,1]}
 }
 {id=WavePlayer_10, properties {soundWaveAssetPtr=/SunRoseFlyingGameplay/Sounds/wav/FN_S29_OW_ChargeBlast_Tell_03}}
 {id=WavePlayer_11, properties {soundWaveAssetPtr=/SunRoseFlyingGameplay/Sounds/wav/FN_S29_OW_ChargeBlast_Tell_04}}
 {id=WavePlayer_12, properties {soundWaveAssetPtr=/SunRoseFlyingGameplay/Sounds/wav/FN_S29_OW_ChargeBlast_Tell_05}}
 {id=WavePlayer_13, properties {soundWaveAssetPtr=/SunRoseFlyingGameplay/Sounds/wav/FN_S29_OW_ChargeBlast_Tell_06}}
 {id=WavePlayer_16, properties {soundWaveAssetPtr=/SunRoseFlyingGameplay/Sounds/wav/FN_S29_OW_ChargeBlast_Tell_01}}
 {id=WavePlayer_17, properties {soundWaveAssetPtr=/SunRoseFlyingGameplay/Sounds/wav/FN_S29_OW_ChargeBlast_Tell_02}}
 {id=WavePlayer_18, properties {soundWaveAssetPtr=&FN_S29_OW_ChargeBlast_Tell_03}}
 {id=WavePlayer_19, properties {soundWaveAssetPtr=&FN_S29_OW_ChargeBlast_Tell_04}}
 {id=WavePlayer_20, properties {soundWaveAssetPtr=&FN_S29_OW_ChargeBlast_Tell_05}}
 {id=WavePlayer_21, properties {soundWaveAssetPtr=&FN_S29_OW_ChargeBlast_Tell_06}}
 {id=WavePlayer_8, properties {soundWaveAssetPtr=&FN_S29_OW_ChargeBlast_Tell_01}}
 {id=WavePlayer_9, properties {soundWaveAssetPtr=&FN_S29_OW_ChargeBlast_Tell_02}}
]
properties {
 bOverrideConcurrency=true
 maxDistance=18000
 soundClassObject=&Fort_Gliders_InAir
 volumeMultiplier=0.45
 concurrencyOverrides {bLimitToOwner=true, maxCount=1, resolutionRule=PreventNew, retriggerTime=2}
}
