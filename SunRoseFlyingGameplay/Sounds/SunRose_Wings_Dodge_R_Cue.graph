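# Sound cue for the right dodge roll of the SunRose wings flying gameplay, classed as an in-air glider sound.
# A Random node picks one of five equally weighted roll waves, feeding a Modulator that varies volume and pitch.
# The PlayerAffiliation switch at the root routes all four inputs to that same Modulator, so every affiliation shares one sound.
# It uses dodge attenuation and owner-limited concurrency with volume scaling.
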SoundCue /SunRoseFlyingGameplay/Sounds/SunRose_Wings_Dodge_R_Cue
root=Switch_0
nodes [
 {id=Modulator_1, children [Random_0], properties {pitchMax=1, volumeMax=0.9, volumeMin=0.7}}
 {id=Random_0, children [WavePlayer_0,WavePlayer_1,WavePlayer_2,WavePlayer_3,WavePlayer_4], properties {weights [1,1,1,1,1]}}
 {id=Switch_0, children [Modulator_1,Modulator_1,Modulator_1,Modulator_1], properties {intParameterName=PlayerAffiliation}}
 {id=WavePlayer_0, properties {soundWaveAssetPtr=/SunRoseFlyingGameplay/Sounds/wav/FN_S29_OW_Roll_R_01}}
 {id=WavePlayer_1, properties {soundWaveAssetPtr=/SunRoseFlyingGameplay/Sounds/wav/FN_S29_OW_Roll_R_02}}
 {id=WavePlayer_2, properties {soundWaveAssetPtr=/SunRoseFlyingGameplay/Sounds/wav/FN_S29_OW_Roll_R_03}}
 {id=WavePlayer_3, properties {soundWaveAssetPtr=/SunRoseFlyingGameplay/Sounds/wav/FN_S29_OW_Roll_R_04}}
 {id=WavePlayer_4, properties {soundWaveAssetPtr=/SunRoseFlyingGameplay/Sounds/wav/FN_S29_OW_Roll_R_05}}
]
properties {
 attenuationSettings=&/SunRose_Wings_Dodge_Att
 bOverrideConcurrency=true
 maxDistance=19900
 pitchMultiplier=0.98
 soundClassObject=&Fort_Gliders_InAir
 volumeMultiplier=0.23
 attenuationOverrides {distanceAlgorithm=NaturalSound, falloffDistance=3000, attenuationShapeExtents {x=1000}}
 concurrencyOverrides {bLimitToOwner=true, maxCount=4, retriggerTime=0.2, voiceStealReleaseTime=0.2, volumeScale=0.6, volumeScaleAttackTime=0.2}
 concurrencySet [&Player_Melee_Weapon_Conc]
}
