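# Sound cue for the SunRose Wings drop impact, playing a landing sound as weapon foley.
# A random node picks between two equally weighted landed waves, feeding a modulator that pins pitch at normal.
# It uses weapon-drop attenuation and a concurrency override that blocks new instances once a small cap is reached.
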SoundCue /SunRoseFlyingGameplay/Sounds/SunRose_Wings_DropImpact_Cue
root=Modulator_1
nodes [
 {id=Modulator_1, children [Random_0], properties {pitchMax=1, pitchMin=1}}
 {id=Random_0, children [WavePlayer_0,WavePlayer_1], properties {weights [1,1]}}
 {id=WavePlayer_0, properties {soundWaveAssetPtr=/SunRoseFlyingGameplay/Sounds/wav/FN_S29_OW_Landed_02}}
 {id=WavePlayer_1, properties {soundWaveAssetPtr=/SunRoseFlyingGameplay/Sounds/wav/FN_S29_OW_Landed_01}}
]
properties {
 attenuationSettings=&Fort_Weapon_Drop_Att
 bOverrideConcurrency=true
 maxDistance=3000
 soundClassObject=&Fort_WeaponFoley
 volumeMultiplier=0.301
 attenuationOverrides {falloffDistance=1000, attenuationShapeExtents {x=0}}
 concurrencyOverrides {maxCount=3, resolutionRule=PreventNew}
}
