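# SoundCue for the drop sound of the SunRose Wings item in the SunRose flying gameplay feature.
# A Random node picks between two drop wave variants with equal weight.
# Playback uses the shared weapon-drop attenuation and stop-oldest concurrency set, routed to the weapon foley sound class at reduced volume.
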SoundCue /SunRoseFlyingGameplay/Sounds/SunRose_Wings_Drop_Cue
root=Random_0
nodes [
 {id=Random_0, children [WavePlayer_2,WavePlayer_3], properties {weights [1,1]}}
 {id=WavePlayer_2, properties {soundWaveAssetPtr=/SunRoseFlyingGameplay/Sounds/wav/FN_S29_OW_Drop_01}}
 {id=WavePlayer_3, properties {soundWaveAssetPtr=/SunRoseFlyingGameplay/Sounds/wav/FN_S29_OW_Drop_02}}
]
properties {
 attenuationSettings=&Fort_Weapon_Drop_Att
 maxDistance=3000
 soundClassObject=&Fort_WeaponFoley
 volumeMultiplier=0.401
 concurrencyOverrides {resolutionRule=PreventNew}
 concurrencySet [&Weapon_Drop_StopOldest]
}
