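# Sound cue for equipping the SunRose wings, routed to the weapon foley sound class.
# A Random node picks one of two equally weighted equip waves and feeds a Modulator that varies each play.
# Spatialization uses the shared generic player-equip attenuation asset, while a local concurrency override caps simultaneous instances and rejects new ones over the limit.
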SoundCue /SunRoseFlyingGameplay/Sounds/SunRose_Wings_Equip_Cue
root=Modulator_2
nodes [
 {id=Modulator_2, children [Random_0], properties {volumeMin=1}}
 {id=Random_0, children [WavePlayer_1,WavePlayer_2], properties {weights [1,1]}}
 {id=WavePlayer_1, properties {soundWaveAssetPtr=/SunRoseFlyingGameplay/Sounds/wav/FN_S29_OW_Equip_01}}
 {id=WavePlayer_2, properties {soundWaveAssetPtr=/SunRoseFlyingGameplay/Sounds/wav/FN_S29_OW_Equip_02}}
]
properties {
 attenuationSettings=/Game/Sounds/Attenuation/Fort_Player_Equip_Generic_Att
 bOverrideConcurrency=true
 maxDistance=1800
 soundClassObject=&Fort_WeaponFoley
 volumeMultiplier=1.103
 attenuationOverrides {distanceAlgorithm=NaturalSound, falloffDistance=3000, attenuationShapeExtents {x=1000}}
 concurrencyOverrides {bVolumeScaleCanRelease=true, maxCount=5, resolutionRule=PreventNew, retriggerTime=0.2, voiceStealReleaseTime=0.2}
}
