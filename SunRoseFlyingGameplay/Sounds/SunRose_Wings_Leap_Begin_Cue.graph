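# Sound cue for the start of a wing leap in SunRose flying gameplay, layering wing-open sounds with a later flap.
# A mixer combines a random open-start wave, a delayed quieter random open-end wave, and a later, pitched-up random flap.
# A quality node gives the first two levels this full mix and the last two a single start wave each.
# The PlayerAffiliation switch routes all four branches to that same quality node, so affiliation makes no difference.
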SoundCue /SunRoseFlyingGameplay/Sounds/SunRose_Wings_Leap_Begin_Cue
root=Modulator_0
nodes [
 {id=Delay_0, children [Random_0], properties {delayMax=0.3, delayMin=0.3}}
 {id=Delay_1, children [Random_2], properties {delayMax=1.2, delayMin=1.2}}
 {id=Mixer_0, children [Modulator_2,Modulator_1,Modulator_3], properties {inputVolume [1,1,1]}}
 {id=Modulator_0, children [Switch_0]}
 {id=Modulator_1, children [Delay_0], properties {volumeMax=0.8, volumeMin=0.8}}
 {id=Modulator_2, children [Random_1], properties {pitchMax=1, pitchMin=1, volumeMax=2, volumeMin=2}}
 {id=Modulator_3, children [Delay_1], properties {pitchMax=1.1, pitchMin=1.1, volumeMax=2, volumeMin=2}}
 {id=QualityLevel_0, children [Mixer_0,Mixer_0,WavePlayer_4,WavePlayer_5]}
 {id=Random_0, children [WavePlayer_1,WavePlayer_2,WavePlayer_6], properties {weights [1,1,1]}}
 {id=Random_1, children [WavePlayer_0,WavePlayer_3,WavePlayer_4,WavePlayer_5], properties {weights [1,1,1,1]}}
 {id=Random_2, children [WavePlayer_7,WavePlayer_8,WavePlayer_9], properties {weights [1,1,1]}}
 {id=Switch_0, children [QualityLevel_0, QualityLevel_0, QualityLevel_0, QualityLevel_0], properties {intParameterName=PlayerAffiliation}}
 {id=WavePlayer_0, properties {soundWaveAssetPtr=/SunRoseFlyingGameplay/Sounds/wav/FN_S29_OW_SpawnOpen_Start_01}}
 {id=WavePlayer_1, properties {soundWaveAssetPtr=/SunRoseFlyingGameplay/Sounds/wav/FN_S29_OW_SpawnOpen_End_01}}
 {id=WavePlayer_2, properties {soundWaveAssetPtr=/SunRoseFlyingGameplay/Sounds/wav/FN_S29_OW_SpawnOpen_End_02}}
 {id=WavePlayer_3, properties {soundWaveAssetPtr=/SunRoseFlyingGameplay/Sounds/wav/FN_S29_OW_SpawnOpen_Start_02}}
 {id=WavePlayer_4, properties {soundWaveAssetPtr=/SunRoseFlyingGameplay/Sounds/wav/FN_S29_OW_SpawnOpen_Start_03}}
 {id=WavePlayer_5, properties {soundWaveAssetPtr=/SunRoseFlyingGameplay/Sounds/wav/FN_S29_OW_SpawnOpen_Start_04}}
 {id=WavePlayer_6, properties {soundWaveAssetPtr=/SunRoseFlyingGameplay/Sounds/wav/FN_S29_OW_SpawnOpen_End_03}}
 {id=WavePlayer_7, properties {soundWaveAssetPtr=/SunRoseFlyingGameplay/Sounds/wav/FN_S29_OW_Flap_05}}
 {id=WavePlayer_8, properties {soundWaveAssetPtr=/SunRoseFlyingGameplay/Sounds/wav/FN_S29_OW_Flap_03}}
 {id=WavePlayer_9, properties {soundWaveAssetPtr=/SunRoseFlyingGameplay/Sounds/wav/FN_S29_OW_Flap_01}}
]
properties {
 attenuationSettings=&/SunRose_Spawn_Att
 bDebug=true
 maxDistance=18000
 soundClassObject=&Fort_WeaponMelee
 volumeMultiplier=0.4
 attenuationOverrides {distanceAlgorithm=NaturalSound, falloffDistance=3000, attenuationShapeExtents {x=1000}}
 concurrencySet [&Player_Melee_Weapon_Conc]
}
