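# SunRose_Wings_Leap_End_Cue is a SoundCue in SunRose flying gameplay, named for the end of a wing leap.
# It holds no sound nodes and no root, so the file defines no audio source.
# It carries playback settings only: lowered pitch, the Fort_Gliders sound class, and a wings attenuation asset alongside natural-sound falloff override values.
# Its concurrency set is borrowed from player melee weapons.
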SoundCue /SunRoseFlyingGameplay/Sounds/SunRose_Wings_Leap_End_Cue
root=null
nodes []
properties {
 attenuationSettings=&/SunRose_Wings_Att
 maxDistance=5300
 pitchMultiplier=0.8
 soundClassObject=/Game/Sounds/FortSoundClassesAndModes/Fort_Gliders
 attenuationOverrides {distanceAlgorithm=NaturalSound, falloffDistance=3000, attenuationShapeExtents {x=1000}}
 concurrencySet [&Player_Melee_Weapon_Conc]
}
