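# Sound cue for the SunRose Wings pickup sound in the SunRose flying gameplay.
# A Random node picks between two equally weighted pickup waves and feeds a Modulator that varies playback, with its minimum pitch held at unshifted.
# It uses shared Fortnite weapon-pickup attenuation and the weapon foley sound class, with voice limiting referencing the PickupMaximum_Oldest concurrency set.
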SoundCue /SunRoseFlyingGameplay/Sounds/SunRose_Wings_Pickup_Cue
root=Modulator_5
nodes [
 {id=Modulator_5, children [Random_0], properties {pitchMin=1}}
 {id=Random_0, children [WavePlayer_0,WavePlayer_1], properties {weights [1,1]}}
 {id=WavePlayer_0, properties {soundWaveAssetPtr=/SunRoseFlyingGameplay/Sounds/wav/FN_S29_OW_Pickup_01}}
 {id=WavePlayer_1, properties {soundWaveAssetPtr=/SunRoseFlyingGameplay/Sounds/wav/FN_S29_OW_Pickup_02}}
]
properties {
 attenuationSettings=&Fort_Weapon_Pickup_Att
 maxDistance=3500
 soundClassObject=&Fort_WeaponFoley
 volumeMultiplier=1.002
 concurrencyOverrides {maxCount=1, resolutionRule=PreventNew}
 concurrencySet [&PickupMaximum_Oldest]
}
