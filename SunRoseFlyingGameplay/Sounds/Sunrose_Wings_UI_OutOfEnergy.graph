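# Sound cue for the Sunrose Wings out-of-energy UI alert in the flying gameplay feature.
# A mixer layers the energy-empty UI wave with a fumes wave that starts after a short fixed delay and plays slightly louder.
# It uses the non-critical in-game UI sound class with overridden distance attenuation.
# Concurrency allows a single instance, blocks new ones while it plays, and sets a retrigger time to limit repeats.
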SoundCue /SunRoseFlyingGameplay/Sounds/Sunrose_Wings_UI_OutOfEnergy
root=Mixer_0
nodes [
 {id=Delay_0, children [WavePlayer_0], properties {delayMax=0.15, delayMin=0.15}}
 {id=Mixer_0, children [WavePlayer_1,Delay_0], properties {inputVolume [1,1.25]}}
 {id=WavePlayer_0, properties {soundWaveAssetPtr=/SunRoseFlyingGameplay/Sounds/wav/FN_S29_OW_Energy_Fumes}}
 {id=WavePlayer_1, properties {soundWaveAssetPtr=/SunRoseFlyingGameplay/Sounds/wav/FN_S29_OW_Energy_UI_Empty}}
]
properties {
 attenuationSettings=/Game/Sounds/Attenuation/Fort_GadgetWorking_Att
 bOverrideAttenuation=true
 bOverrideConcurrency=true
 maxDistance=1536
 soundClassObject=&Fort_UI_InGame_NonCritical
 volumeMultiplier=1
 attenuationOverrides {falloffDistance=1024, attenuationShapeExtents {x=512}}
 concurrencyOverrides {maxCount=1, resolutionRule=PreventNew, retriggerTime=2}
}
