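# Sound cue for the SunRose wings gameplay that, judging by its name, plays as UI feedback when wing energy is replenished.
# A single Wave Player plays the Energy Replenished wave, routed to the non-critical in-game UI sound class with a slight volume boost.
# Despite the UI class, attenuation overrides give it a distance falloff, and a concurrency override caps simultaneous instances.
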SoundCue /SunRoseFlyingGameplay/Sounds/Sunrose_Wings_UI_Replenish_Cue
root=WavePlayer_5
nodes [{id=WavePlayer_5, properties {soundWaveAssetPtr=/SunRoseFlyingGameplay/Sounds/wav/FN_S29_OW_Energy_Replenished}}]
properties {
 attenuationSettings=/Game/Sounds/Attenuation/Fort_GadgetWorking_Att
 bOverrideAttenuation=true
 bOverrideConcurrency=true
 maxDistance=1536
 soundClassObject=&Fort_UI_InGame_NonCritical
 volumeMultiplier=1.4
 attenuationOverrides {falloffDistance=1024, attenuationShapeExtents {x=512}}
 concurrencyOverrides {maxCount=3}
}
